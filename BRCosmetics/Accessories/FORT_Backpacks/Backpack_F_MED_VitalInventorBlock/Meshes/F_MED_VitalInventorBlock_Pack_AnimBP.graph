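# Animation Blueprint for the Vital Inventor Block backpack that adds secondary physics motion to its left and right ear bones.
# Four chained Anim Dynamics nodes process the input pose, two per ear, with angular springs and limits.
# The bone-relative pair is active when not skydive floating, and the root-relative pair with different gravity takes over while floating.
# Update Animation derives not_floating by inverting bIsSkydiveFloating, and timed alpha blends smooth the switch.
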AnimBlueprint /BRCosmetics/Accessories/FORT_Backpacks/Backpack_F_MED_VitalInventorBlock/Meshes/F_MED_VitalInventorBlock_Pack_AnimBP
target_skeleton=&/F_MED_VitalInventorBlock_Pack_Skeleton
graphs [
 AnimGraph {nodes [
   n1 = Get not_floating() -> (not_floating:bool)
   n2 = Anim Dynamics - Bone: Dyn_l_ear!(
    ComponentPose=Local To Component(LocalPose=Input Pose())
    bAlphaBoolEnabled:bool=@n1
    GravityOverride:Vector="0,0,100"
    ExternalForce="0,0,0"
   ) -> (Pose) {
    alphaInputType=Bool
    angularDampingOverride=1
    angularSpringConstant=1000
    bAngularSpring=true
    bGravityOverrideInSimSpace=true
    bOverrideAngularDamping=true
    bUseGravityOverride=true
    bUsePlanarLimit=false
    simulationSpace=BoneRelative
    alphaBoolBlend {blendInTime=0.2, blendOption=Linear, blendOutTime=0.2}
    boundBone {boneName=Dyn_l_ear}
    componentAppliedLinearAccClamp=<1000 1000 1000>
    componentLinearAccScale=<0.2 0.2 0.2>
    componentLinearVelScale=<0.3 0.3 0.3>
    gravityOverride {z=-500}
    physicsBodyDefinitions [{
      collisionType=CustomSphere
      sphereCollisionRadius=2.5
      boundBone {boneName=Dyn_l_ear}
      boxExtents=<6 6 6>
      constraintSetup {angularLimitsMax=<3 10 10>, angularLimitsMin=<-3 -10 -15>, angularTarget=<0 0 0>}
      localJointOffset=<3 0 0.5>
     }]
    relativeSpaceBone {boneName=pack}
   }
   n3 = Anim Dynamics - Bone: Dyn_r_ear!(ComponentPose=@n2, bAlphaBoolEnabled:bool=@n1, GravityOverride:Vector="0,0,100", ExternalForce="0,0,0") -> (Pose) {
    alphaInputType=Bool
    angularDampingOverride=1
    angularSpringConstant=1000
    bAngularSpring=true
    bGravityOverrideInSimSpace=true
    bOverrideAngularDamping=true
    bUseGravityOverride=true
    bUsePlanarLimit=false
    simulationSpace=BoneRelative
    alphaBoolBlend {blendInTime=0.2, blendOption=Linear, blendOutTime=0.4}
    boundBone {boneName=Dyn_r_ear}
    componentAppliedLinearAccClamp=<1000 1000 1000>
    componentLinearAccScale=<0.2 0.2 0.2>
    componentLinearVelScale=<0.3 0.3 0.3>
    gravityOverride {z=-500}
    physicsBodyDefinitions [{
      collisionType=CustomSphere
      sphereCollisionRadius=2.5
      boundBone {boneName=Dyn_r_ear}
      boxExtents=<6 6 6>
      constraintSetup {angularLimitsMax=<3 10 10>, angularLimitsMin=<-3 -10 -15>, angularTarget=<0 0 0>}
      localJointOffset=<3 0 -0.5>
     }]
    relativeSpaceBone {boneName=pack}
   }
   n4 = Get bIsSkydiveFloating()
   n5 = Anim Dynamics - Bone: Dyn_l_ear(ComponentPose=@n3, bAlphaBoolEnabled=@n4, GravityOverride="0,0,-600", ExternalForce="0,0,0") {
    alphaInputType=Bool
    angularDampingOverride=0.9999
    angularSpringConstant=1000
    bAngularSpring=true
    bOverrideAngularDamping=true
    bUseGravityOverride=true
    bUsePlanarLimit=false
    simulationSpace=RootRelative
    alphaBoolBlend {blendInTime=0.2, blendOption=Custom, blendOutTime=0.4}
    boundBone {boneName=Dyn_l_ear}
    componentAppliedLinearAccClamp=<3000 3000 3000>
    componentLinearAccScale=<0.1 0.1 0.1>
    componentLinearVelScale=<0.2 0.2 0.2>
    gravityOverride {z=-500}
    physicsBodyDefinitions [{
      collisionType=CustomSphere
      sphereCollisionRadius=2.5
      boundBone {boneName=Dyn_l_ear}
      boxExtents=<12 12 12>
      constraintSetup {angularLimitsMax=<3 12 0>, angularLimitsMin=<-3 -12 -15>, angularTarget=<0 0 0>}
      localJointOffset=<0.5 0 0>
     }]
    relativeSpaceBone {boneName=pack}
   }
   n6 = Anim Dynamics - Bone: Dyn_r_ear(ComponentPose=@n5, bAlphaBoolEnabled=@n4, GravityOverride="0,0,-600", ExternalForce="0,0,0") {
    alphaInputType=Bool
    angularDampingOverride=0.9999
    angularSpringConstant=1000
    bAngularSpring=true
    bOverrideAngularDamping=true
    bUseGravityOverride=true
    bUsePlanarLimit=false
    simulationSpace=RootRelative
    alphaBoolBlend {blendInTime=0.2, blendOption=Custom, blendOutTime=0.4}
    boundBone {boneName=Dyn_r_ear}
    componentAppliedLinearAccClamp=<3000 3000 3000>
    componentLinearAccScale=<0.1 0.1 0.1>
    componentLinearVelScale=<0.2 0.2 0.2>
    gravityOverride {z=-500}
    physicsBodyDefinitions [{
      collisionType=CustomSphere
      sphereCollisionRadius=2.5
      boundBone {boneName=Dyn_r_ear}
      boxExtents=<12 12 12>
      constraintSetup {angularLimitsMax=<3 12 0>, angularLimitsMin=<-3 -12 -15>, angularTarget=<0 0 0>}
      localJointOffset=<0.5 0 0>
     }]
    relativeSpaceBone {boneName=pack}
   }
   Output Pose(Result=Component To Local(ComponentPose=@n6)) {layerGroup=None}
  ]}
 EventGraph {nodes [
   Event BlueprintUpdateAnimation()
   > Set not_floating(not_floating:bool=NOT Boolean(A=Get bIsSkydiveFloating())) -> (Output_Get:bool)
   TryGetPawnOwner()
  ]}
]
variables [not_floating {type=boolean}]
